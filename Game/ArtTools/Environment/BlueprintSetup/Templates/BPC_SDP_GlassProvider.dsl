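; Template component Blueprint named as an SDP glass provider; the only logic it shows removes it from unsupported environments (FNC, VKEdit, VKPlay).
; OnConstruct checks IsSpatialGameplayExperience and uses a next-frame delay and DestroyComponent for that cleanup, ignoring its bIsUEFN input.
; BeginPlay and Tick are empty, and no glass-specific behavior appears in the graph.
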(event EventBeginPlay)

(event EventTick
  (params (DeltaSeconds :type Float%20%28single-precision%29)))

(event EventOnConstruct
  (params (bIsUEFN :type Boolean))
  ;[1 1 1] Delete component in unsupported environments (FNC, VKEdit, VKPlay)
  (if (Fortnite|IsSpatialGameplayExperience)
    (Utilities|FlowControl|DelayUntilNextFrame)
    (Components|DestroyComponent)))
